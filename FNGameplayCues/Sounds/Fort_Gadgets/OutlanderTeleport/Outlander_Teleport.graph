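# Sound cue for the Outlander teleport gadget effect, layering blink and phase-shift teleport sounds.
# A mixer blends a random pick of four blink waves, passed through a pitch modulator, with a quieter random pick of three phase-shift waves.
# It reuses small melee weapon attenuation and a melee sound class, primes on load, and overrides concurrency to prevent new instances at the limit.
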SoundCue /FNGameplayCues/Sounds/Fort_Gadgets/OutlanderTeleport/Outlander_Teleport
root=Mixer_0
nodes [
 {id=Mixer_0, children [Modulator_2,Random_0], properties {inputVolume [1,0.35]}}
 {id=Modulator_2, children [Random_1], properties {pitchMax=1.1}}
 {id=Random_0, children [WavePlayer_4,WavePlayer_5,WavePlayer_6], properties {weights [1,1,1]}}
 {id=Random_1, children [WavePlayer_0,WavePlayer_1,WavePlayer_2,WavePlayer_3], properties {weights [1,1,1,1]}}
 {id=WavePlayer_0, properties {soundWaveAssetPtr=&/Outlander_Blink_Teleport_FX_003}}
 {id=WavePlayer_1, properties {soundWaveAssetPtr=&/Outlander_Blink_Teleport_FX_002}}
 {id=WavePlayer_2, properties {soundWaveAssetPtr=&/Outlander_Blink_Teleport_FX_001}}
 {id=WavePlayer_3, properties {soundWaveAssetPtr=&/Outlander_Blink_Teleport_FX_000}}
 {id=WavePlayer_4, properties {soundWaveAssetPtr=&/Phase_Shift_Teleport_01}}
 {id=WavePlayer_5, properties {soundWaveAssetPtr=&/Phase_Shift_Teleport_02}}
 {id=WavePlayer_6, properties {soundWaveAssetPtr=&/Phase_Shift_Teleport_03}}
]
properties {
 attenuationSettings=&Fort_Weapon_Melee_Sml_Att
 bOverrideConcurrency=true
 bPrimeOnLoad=true
 maxDistance=2800
 soundClassObject=&Fort_WeaponMelee_PSM
 volumeMultiplier=1.25
 concurrencyOverrides {resolutionRule=PreventNew}
}
